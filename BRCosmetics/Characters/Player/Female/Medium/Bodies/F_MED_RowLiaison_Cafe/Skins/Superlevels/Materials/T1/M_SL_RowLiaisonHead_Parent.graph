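# Parent material for the Superlevels skin head of the female RowLiaison Cafe character, adding a tintable lip recolor.
# It wraps the edit between the shared 2023 character PreFX and PostFX functions, changing only Base Color in the material attributes.
# Base color is desaturated, scaled by brightness and tinted, then blended in through the FX2 texture's green channel.
# The Lips_Multiplier parameter feeds only an unconnected OneMinus, so it has no effect.
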Material /BRCosmetics/Characters/Player/Female/Medium/Bodies/F_MED_RowLiaison_Cafe/Skins/Superlevels/Materials/T1/M_SL_RowLiaisonHead_Parent
graph {
 properties {
  bAutomaticallySetUsageInEditor=false
  bUseMaterialAttributes=true
  bUsedWithClothing=true
  bUsedWithMeshDeformer=true
  bUsedWithMorphTargets=true
  bUsedWithNanite=true
  bUsedWithNiagaraMeshParticles=true
  bUsedWithSkeletalMesh=true
  pixelDepthOffsetMode=PDOM_Legacy
  shadingModel=MSM_FromMaterialExpression
  usedShadingModels=Default Lit | Subsurface | Clear Coat | Subsurface Profile | Cloth
 }
 nodes [
  n1 = MaterialFunctionCall() {materialFunction=&MF_CharacterParent_2023_PreFX}
  n2 = GetMaterialAttributes(None=@n1.Result) {attributeGetTypes [#1]}
  n3 = ScalarParameter() {defaultValue=1, group=Superlevel, parameterName=Lips_BaseColorDesaturation}
  n4 = Multiply(
   A=Multiply(A=Desaturation(None=@n2.Base Color, Fraction=@n3), B=ScalarParameter() {defaultValue=1, group=Superlevel, parameterName=Lips_Brightness})
   B=VectorParameter() {group=Superlevel, parameterName=Lips_Color, defaultValue=<rgba 1 1 1 0>}.RGB
  )
  n5 = TextureSample() {
   samplerType=SAMPLERTYPE_Masks
   texture=/BRCosmetics/Characters/Player/Female/Medium/Bodies/F_MED_RowLiaison_Cafe/Textures/T_RowLiaison_Cafe_Head_SL_FX2
  }
  MaterialFunctionCall(
   Material=SetMaterialAttributes!(MaterialAttributes=@n2.MaterialAttributes, Base Color=LinearInterpolate(A=@n2.Base Color, B=@n4, Alpha=@n5.G)) {attributeSetTypes [#1]}
   M=@n1.M
   Pre Skinned Local Position=@n1.M
   Distance Blend=@n1.M
   Fresnel=@n1.M
  ) {materialFunction=&MF_CharacterParent_2023_PostFX}
  OneMinus(None=ScalarParameter() {defaultValue=1, group=Superlevel, parameterName=Lips_Multiplier})
 ]
}
